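# Physics asset for the left-leg straps on the medium male KelpLinenMagnesium player body, providing capsule bodies for the straps and both legs.
# A three-segment belt chain from belt_base_l and a single segment from corrective_belt_l have narrow swing and twist limits.
# Both belt bases join the left calf with free rotation, as do the calf-thigh-root links.
# A huge zero-length root capsule sits below the origin, approximating a floor.
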PhysicsAsset /BRCosmetics/Characters/Player/Male/Medium/Bodies/M_MED_KelpLinenMagnesium/Meshes/M_MED_KelpLinenMagnesium_Left_Leg_Strapes_Physics
bodies {
 belt_base_l [{shapeType=Capsule, length=1, radius=1}]
 calf_l [{shapeType=Capsule, length=22.8198, radius=8.5, center=<-25.3339 -1.6397 0.6544>, rotation=<rot 0 -89.532 -89.8784>}]
 calf_r [{shapeType=Capsule, length=23.7802, radius=7.5, center=<19.7095 1.4284 -0.7123>, rotation=<rot 0 -90.0261 -89.6633>}]
 corrective_belt_l [{shapeType=Capsule, length=0.505, radius=0.505, shapeName=corrective_belt_l_capsule, rotation=<rot 0 -90 0>}]
 dyn_belt_l [{
   shapeType=Capsule
   length=2.9445
   radius=2.5608
   shapeName=dyn_belt_l_capsule
   center=<5.6504 0.1027 1.355>
   rotation=<rot 7.5381 95.1464 -77.702>
  }]
 dyn_belt_l_01 [{
   shapeType=Capsule
   length=8.1682
   radius=1.8601
   shapeName=dyn_belt_l_01_capsule
   center=<6.7139 -0.1609 0.2406>
   rotation=<rot -0.4457 -91.3565 -92.0522>
  }]
 dyn_belt_l_02 [{
   shapeType=Capsule
   length=5.4037
   radius=1.7793
   shapeName=dyn_belt_l_02_capsule
   center=<4.8945 -0.1706 0.1973>
   rotation=<rot 0 88.0034 -87.6935>
  }]
 dyn_belt_l_03 [{
   shapeType=Capsule
   length=5.2553
   radius=1.7217
   shapeName=dyn_belt_l_03_capsule
   center=<2.8897 0.1299 1.418>
   rotation=<rot -3.7257 89.5381 -65.9941>
  }]
 root [{shapeType=Capsule, length=0, radius=1000, shapeName=root_capsule, center=<0 0 -998.8453>, rotation=<rot 0 -90 0>}]
 thigh_l [{shapeType=Capsule, length=27.0058, radius=8, shapeName=thigh_l_capsule, center=<-18.564 0 0>, rotation=<rot 90 0 0>}]
 thigh_r [{shapeType=Capsule, length=27.162, radius=8, shapeName=thigh_r_capsule, center=<18.564 0 0>, rotation=<rot -90 0 0>}]
}
constraints [
 {
  bone1Name=belt_base_l
  bone2Name=calf_l
  swing1LimitDegrees=45
  swing1Motion=Free
  swing2LimitDegrees=45
  swing2Motion=Free
  twistLimitDegrees=45
  twistMotion=Free
 }
 {bone1Name=dyn_belt_l_01, bone2Name=belt_base_l, swing1LimitDegrees=15, swing2LimitDegrees=30, twistLimitDegrees=2}
 {bone1Name=dyn_belt_l_02, bone2Name=dyn_belt_l_01, swing1LimitDegrees=30, swing2LimitDegrees=30, twistLimitDegrees=2}
 {bone1Name=dyn_belt_l_03, bone2Name=dyn_belt_l_02, swing1LimitDegrees=30, swing2LimitDegrees=30, twistLimitDegrees=2}
 {
  bone1Name=calf_r
  bone2Name=thigh_r
  swing1LimitDegrees=45
  swing1Motion=Free
  swing2LimitDegrees=45
  swing2Motion=Free
  twistLimitDegrees=45
  twistMotion=Free
 }
 {
  bone1Name=calf_l
  bone2Name=thigh_l
  swing1LimitDegrees=45
  swing1Motion=Free
  swing2LimitDegrees=45
  swing2Motion=Free
  twistLimitDegrees=45
  twistMotion=Free
 }
 {
  bone1Name=thigh_l
  bone2Name=root
  swing1LimitDegrees=45
  swing1Motion=Free
  swing2LimitDegrees=45
  swing2Motion=Free
  twistLimitDegrees=45
  twistMotion=Free
 }
 {
  bone1Name=thigh_r
  bone2Name=root
  swing1LimitDegrees=45
  swing1Motion=Free
  swing2LimitDegrees=45
  swing2Motion=Free
  twistLimitDegrees=45
  twistMotion=Free
 }
 {bone1Name=dyn_belt_l, bone2Name=corrective_belt_l, swing1LimitDegrees=5, swing2LimitDegrees=10, twistLimitDegrees=2}
 {
  bone1Name=corrective_belt_l
  bone2Name=calf_l
  swing1LimitDegrees=45
  swing1Motion=Free
  swing2LimitDegrees=45
  swing2Motion=Free
  twistLimitDegrees=45
  twistMotion=Free
 }
]
