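# Physics asset for the right-leg straps of the KelpLinenMagnesium male medium body cosmetic, built from capsule bodies.
# A three-segment belt chain off belt_base_r and a separate dyn_belt_r off corrective_belt_r use tight swing and twist limits, with both anchors tied to calf_r.
# Thigh and calf capsules on both legs connect to root through free constraints.
# The oversized root capsule is offset downward, apparently acting as a ground plane.
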PhysicsAsset /BRCosmetics/Characters/Player/Male/Medium/Bodies/M_MED_KelpLinenMagnesium/Meshes/M_MED_KelpLinenMagnesium_Rights_Leg_Strapes_Physics
bodies {
 belt_base_r [{shapeType=Capsule, length=1, radius=1, shapeName=belt_base_r_capsule}]
 calf_l [{shapeType=Capsule, length=23.8198, radius=7.5, center=<-19.7096 -1.5938 0.6424>, rotation=<rot 0 -89.532 -89.8784>}]
 calf_r [{shapeType=Capsule, length=28.2802, radius=9, center=<23.3462 1.4267 -0.7337>, rotation=<rot 0 -90.0261 -89.6633>}]
 corrective_belt_r [{shapeType=Capsule, length=0.505, radius=0.505, shapeName=corrective_belt_r_capsule, rotation=<rot 0 -90 0>}]
 dyn_belt_r [{
   shapeType=Capsule
   length=4.4815
   radius=1.8092
   shapeName=dyn_belt_r_capsule
   center=<-3.4973 -0.0474 -1.3558>
   rotation=<rot 0 95.5715 -81.1796>
  }]
 dyn_belt_r_01 [{shapeType=Capsule, length=8.3435, radius=1.8757, center=<6.9235 0.356 -0.0606>, rotation=<rot 0 91.8533 -89.1869>}]
 dyn_belt_r_02 [{
   shapeType=Capsule
   length=8.7934
   radius=2.6461
   shapeName=dyn_belt_r_02_capsule
   center=<6.175 -0.0306 -0.0791>
   rotation=<rot 0 -90.2842 -89.2662>
  }]
 dyn_belt_r_03 [{
   shapeType=Capsule
   length=4.8567
   radius=2.6373
   shapeName=dyn_belt_r_03_capsule
   center=<4.1475 -0.5732 -2.395>
   rotation=<rot -14.3624 90.5544 -120.8337>
  }]
 root [{shapeType=Capsule, length=0, radius=1000, shapeName=root_capsule, center=<0 0 -998.8453>, rotation=<rot 0 -90 0>}]
 thigh_l [{shapeType=Capsule, length=27.0058, radius=8, shapeName=thigh_l_capsule, center=<-18.564 0 0>, rotation=<rot 90 0 0>}]
 thigh_r [{shapeType=Capsule, length=27.162, radius=8, shapeName=thigh_r_capsule, center=<18.564 0 0>, rotation=<rot -90 0 0>}]
}
constraints [
 {bone1Name=dyn_belt_r_01, bone2Name=belt_base_r, swing1LimitDegrees=15, swing2LimitDegrees=30, twistLimitDegrees=2}
 {bone1Name=dyn_belt_r_02, bone2Name=dyn_belt_r_01, swing1LimitDegrees=30, swing2LimitDegrees=30, twistLimitDegrees=2}
 {bone1Name=dyn_belt_r_03, bone2Name=dyn_belt_r_02, swing1LimitDegrees=30, swing2LimitDegrees=30, twistLimitDegrees=2}
 {
  bone1Name=belt_base_r
  bone2Name=calf_r
  swing1LimitDegrees=45
  swing1Motion=Free
  swing2LimitDegrees=45
  swing2Motion=Free
  twistLimitDegrees=45
  twistMotion=Free
 }
 {
  bone1Name=calf_r
  bone2Name=thigh_r
  swing1LimitDegrees=45
  swing1Motion=Free
  swing2LimitDegrees=45
  swing2Motion=Free
  twistLimitDegrees=45
  twistMotion=Free
 }
 {
  bone1Name=calf_l
  bone2Name=thigh_l
  swing1LimitDegrees=45
  swing1Motion=Free
  swing2LimitDegrees=45
  swing2Motion=Free
  twistLimitDegrees=45
  twistMotion=Free
 }
 {
  bone1Name=thigh_l
  bone2Name=root
  swing1LimitDegrees=45
  swing1Motion=Free
  swing2LimitDegrees=45
  swing2Motion=Free
  twistLimitDegrees=45
  twistMotion=Free
 }
 {
  bone1Name=thigh_r
  bone2Name=root
  swing1LimitDegrees=45
  swing1Motion=Free
  swing2LimitDegrees=45
  swing2Motion=Free
  twistLimitDegrees=45
  twistMotion=Free
 }
 {bone1Name=dyn_belt_r, bone2Name=corrective_belt_r, swing1LimitDegrees=5, swing2LimitDegrees=10, twistLimitDegrees=2}
 {
  bone1Name=corrective_belt_r
  bone2Name=calf_r
  swing1LimitDegrees=45
  swing1Motion=Free
  swing2LimitDegrees=45
  swing2Motion=Free
  twistLimitDegrees=45
  twistMotion=Free
 }
]
