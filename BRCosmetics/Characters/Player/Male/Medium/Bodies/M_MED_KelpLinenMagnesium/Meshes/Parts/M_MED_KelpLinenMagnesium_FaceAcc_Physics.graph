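# Physics asset for the face accessory mesh of the male medium KelpLinenMagnesium cosmetic body, setting up secondary motion for its hair bones.
# Capsule bodies on the dyn_hair bones are constrained to the hair root with tight twist and small swing limits.
# The top tuft forms a short chain, with top_02 and top_03 hanging off a nearly rigid top_01.
# An unconstrained right-clavicle capsule is also present, likely as a shoulder collider for the hair.
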PhysicsAsset /BRCosmetics/Characters/Player/Male/Medium/Bodies/M_MED_KelpLinenMagnesium/Meshes/Parts/M_MED_KelpLinenMagnesium_FaceAcc_Physics
bodies {
 clavicle_r [{
   shapeType=Capsule
   length=14.505
   radius=9.505
   shapeName=clavicle_r_capsule
   center=<-5.7221 0.9706 1.5864>
   rotation=<rot -0.0001 -90 98.9667>
  }]
 dyn_hair_a [{
   shapeType=Capsule
   length=2.3989
   radius=1
   shapeName=dyn_hair_a_capsule
   center=<6.0372 -0.2018 -1.675>
   rotation=<rot 0 -104.7721 -54.7499>
  }]
 dyn_hair_b [{shapeType=Capsule, length=3.1726, radius=1, shapeName=dyn_hair_b_capsule, center=<5.1475 0.213 0.3768>, rotation=<rot 0 -67.0226 -87.256>}]
 dyn_hair_c [{shapeType=Capsule, length=4.2635, radius=1, shapeName=dyn_hair_c_capsule, center=<5.9714 0.249 -0.2063>, rotation=<rot 0 -84.2355 -82.8005>}]
 dyn_hair_d [{
   shapeType=Capsule
   length=2.2245
   radius=1
   shapeName=dyn_hair_d_capsule
   center=<4.3404 -0.2395 0.1816>
   rotation=<rot 0 -102.1689 -78.1842>
  }]
 dyn_hair_e [{
   shapeType=Capsule
   length=1.9746
   radius=1
   shapeName=dyn_hair_e_capsule
   center=<4.5612 -0.0813 -0.7947>
   rotation=<rot 0 -90.9428 -73.2475>
  }]
 dyn_hair_f [{
   shapeType=Capsule
   length=2.7237
   radius=1
   shapeName=dyn_hair_f_capsule
   center=<3.4646 -0.0681 -0.8992>
   rotation=<rot 0 -94.5248 -72.9941>
  }]
 dyn_hair_j [{
   shapeType=Capsule
   length=2.9477
   radius=1
   shapeName=dyn_hair_j_capsule
   center=<4.4916 -0.0694 -0.0375>
   rotation=<rot 0 -90.2988 -76.2587>
  }]
 dyn_hair_k [{shapeType=Capsule, length=3.5374, radius=1, shapeName=dyn_hair_k_capsule, center=<5.4538 0.4551 -1.2982>, rotation=<rot 0 -87.7223 -72.88>}]
 dyn_hair_l [{
   shapeType=Capsule
   length=2.8473
   radius=1
   shapeName=dyn_hair_l_capsule
   center=<6.2859 -0.4176 -0.9063>
   rotation=<rot 2.9383 -77.5743 -77.5056>
  }]
 dyn_hair_m [{
   shapeType=Capsule
   length=4.8441
   radius=1
   shapeName=dyn_hair_m_capsule
   center=<5.7019 0.341 -1.7699>
   rotation=<rot 6.4661 -81.4611 -68.2726>
  }]
 dyn_hair_n [{shapeType=Capsule, length=0.505, radius=1, shapeName=dyn_hair_n_capsule, center=<6.6439 0.2632 0.0193>, rotation=<rot 0 -178.9697 -66.7341>}]
 dyn_hair_o [{
   shapeType=Capsule
   length=1.4475
   radius=1
   shapeName=dyn_hair_o_capsule
   center=<4.7549 -0.0317 -1.2888>
   rotation=<rot 0 -102.9135 -61.0552>
  }]
 dyn_hair_p [{
   shapeType=Capsule
   length=4.7336
   radius=1
   shapeName=dyn_hair_p_capsule
   center=<7.4975 -0.8019 -1.8666>
   rotation=<rot 0 -91.5566 -70.5493>
  }]
 dyn_hair_q [{
   shapeType=Capsule
   length=2.6559
   radius=1
   shapeName=dyn_hair_q_capsule
   center=<5.8052 -0.3601 -0.5595>
   rotation=<rot 0 -112.2022 -84.5738>
  }]
 dyn_hair_qq [{shapeType=Capsule, length=3.1373, radius=1, shapeName=dyn_hair_qq_capsule, center=<6.4351 0.2277 0.1184>, rotation=<rot 0 -85.6586 -85.0293>}]
 dyn_hair_r [{shapeType=Capsule, length=2.5085, radius=1, shapeName=dyn_hair_r_capsule, center=<5.6819 -0.2569 0.6302>, rotation=<rot 0 -69.7368 -82.5543>}]
 dyn_hair_s [{shapeType=Capsule, length=1.8857, radius=1, shapeName=dyn_hair_s_capsule, center=<7.1025 -1.1937 0.0389>, rotation=<rot 0 -50.0836 -67.2028>}]
 dyn_hair_t [{shapeType=Capsule, length=3.7967, radius=1, shapeName=dyn_hair_t_capsule, center=<4.7245 0.057 -0.2147>, rotation=<rot 0 97.1131 -89.5728>}]
 dyn_hair_top_01 [{
   shapeType=Capsule
   length=1.3793
   radius=1
   shapeName=dyn_hair_top_01_capsule
   center=<1.5044 -0.2863 0.1839>
   rotation=<rot 26.6072 -90 16.0504>
  }]
 dyn_hair_top_02 [{
   shapeType=Capsule
   length=4.7266
   radius=1
   shapeName=dyn_hair_top_02_capsule
   center=<2.8573 -1.3027 -0.2771>
   rotation=<rot 0 68.9695 -86.1707>
  }]
 dyn_hair_top_03 [{
   shapeType=Capsule
   length=3.0409
   radius=1
   shapeName=dyn_hair_top_03_capsule
   center=<2.7444 -2.0766 1.2681>
   rotation=<rot 0 26.4356 -68.2705>
  }]
 hair [{shapeType=Capsule, length=0, radius=1.6679, center=<0.8952 0.0669 -2.4509>, rotation=<rot 0 11.1596 -45.9423>}]
}
constraints [
 {bone1Name=dyn_hair_a, bone2Name=hair, swing1LimitDegrees=10, swing2LimitDegrees=10, twistLimitDegrees=1}
 {bone1Name=dyn_hair_b, bone2Name=hair, swing1LimitDegrees=20, swing2LimitDegrees=20, twistLimitDegrees=1}
 {bone1Name=dyn_hair_c, bone2Name=hair, swing1LimitDegrees=20, swing2LimitDegrees=20, twistLimitDegrees=1}
 {bone1Name=dyn_hair_d, bone2Name=hair, swing1LimitDegrees=20, swing2LimitDegrees=20, twistLimitDegrees=1}
 {bone1Name=dyn_hair_e, bone2Name=hair, swing1LimitDegrees=20, swing2LimitDegrees=20, twistLimitDegrees=1}
 {bone1Name=dyn_hair_f, bone2Name=hair, swing1LimitDegrees=20, swing2LimitDegrees=20, twistLimitDegrees=1}
 {bone1Name=dyn_hair_j, bone2Name=hair, swing1LimitDegrees=20, swing2LimitDegrees=20, twistLimitDegrees=1}
 {bone1Name=dyn_hair_k, bone2Name=hair, swing1LimitDegrees=20, swing2LimitDegrees=20, twistLimitDegrees=1}
 {bone1Name=dyn_hair_l, bone2Name=hair, swing1LimitDegrees=20, swing2LimitDegrees=20, twistLimitDegrees=1}
 {bone1Name=dyn_hair_m, bone2Name=hair, swing1LimitDegrees=20, swing2LimitDegrees=10, twistLimitDegrees=1}
 {bone1Name=dyn_hair_n, bone2Name=hair, swing1LimitDegrees=10, swing2LimitDegrees=10, twistLimitDegrees=1}
 {bone1Name=dyn_hair_o, bone2Name=hair, swing1LimitDegrees=10, swing2LimitDegrees=10, twistLimitDegrees=1}
 {bone1Name=dyn_hair_p, bone2Name=hair, swing1LimitDegrees=20, swing2LimitDegrees=20, twistLimitDegrees=1}
 {bone1Name=dyn_hair_q, bone2Name=hair, swing1LimitDegrees=20, swing2LimitDegrees=20, twistLimitDegrees=1}
 {bone1Name=dyn_hair_r, bone2Name=hair, swing1LimitDegrees=20, swing2LimitDegrees=20, twistLimitDegrees=1}
 {bone1Name=dyn_hair_s, bone2Name=hair, swing1LimitDegrees=20, swing2LimitDegrees=20, twistLimitDegrees=1}
 {bone1Name=dyn_hair_t, bone2Name=hair, swing1LimitDegrees=20, swing2LimitDegrees=20, twistLimitDegrees=1}
 {bone1Name=dyn_hair_top_01, bone2Name=hair, swing1LimitDegrees=2, swing2LimitDegrees=2, twistLimitDegrees=1}
 {bone1Name=dyn_hair_top_02, bone2Name=dyn_hair_top_01, swing1LimitDegrees=20, swing2LimitDegrees=10, twistLimitDegrees=1}
 {bone1Name=dyn_hair_top_03, bone2Name=dyn_hair_top_01, swing1LimitDegrees=20, swing2LimitDegrees=10, twistLimitDegrees=1}
 {bone1Name=dyn_hair_qq, bone2Name=hair, swing1LimitDegrees=20, swing2LimitDegrees=20, twistLimitDegrees=1}
]
