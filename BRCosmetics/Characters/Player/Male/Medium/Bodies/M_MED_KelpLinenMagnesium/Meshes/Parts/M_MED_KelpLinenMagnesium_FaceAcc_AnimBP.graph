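# Post-process-style AnimBP for the face accessory of the male medium KelpLinenMagnesium cosmetic body, adding secondary physics motion.
# It copies the attached parent mesh's pose, then runs a RigidBody node simulated in head-bone space with its own physics asset and overridden gravity.
# Each update rebuilds the sim space settings, mapping owner actor scale over 500–1000 to world alpha, which stays 1 at normal scales.
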AnimBlueprint /BRCosmetics/Characters/Player/Male/Medium/Bodies/M_MED_KelpLinenMagnesium/Meshes/Parts/M_MED_KelpLinenMagnesium_FaceAcc_AnimBP
target_skeleton=&/M_MED_KelpLinenMagnesium_FaceAcc_Skeleton
graphs [
 AnimGraph {nodes [
   n1 = RigidBody!(
    ComponentPose=Local To Component(LocalPose=Copy Pose From Mesh() {bUseAttachedParent=true})
    Alpha
    ExternalForce
    SimSpaceSettings:Sim Space Settings Structure=Get Sim Space Settings() -> (Sim Space Settings:Sim Space Settings Structure)
   ) -> (Pose) {
    bClampLinearTranslationLimitToRefPose=true
    bOverrideWorldGravity=true
    overridePhysicsAsset=&/M_MED_KelpLinenMagnesium_FaceAcc_Physics
    simulationSpace=BaseBoneSpace
    alphaBoolBlend {blendOption=Linear}
    baseBoneRef {boneName=head}
    componentAppliedLinearAccClamp=<3000 3000 3000>
    componentLinearAccScale=<0.3 0.2 0.2>
    componentLinearVelScale=<0.3 0.2 0.2>
    overrideWorldGravity {z=-980}
    simSpaceSettings {maxAngularAcceleration=1000, maxAngularVelocity=3, maxLinearAcceleration=250, maxLinearVelocity=500, velocityScaleZ=0, worldAlpha=1}
   }
   Output Pose(Result=Component To Local(ComponentPose=@n1)) {layerGroup=None}
  ]}
 EventGraph {nodes [
   Event BlueprintUpdateAnimation()
   n1 = TryGetPawnOwner()
   > Set!(
    execute
    Sim Space Settings_WorldAlpha:float=MapRangeClamped(
     Value=GetActorScale3D!(self:Actor Object Reference=@n1) -> (ReturnValue_X:double, ReturnValue_Y:double, ReturnValue_Z:double)
     InRangeA=500
     InRangeB=1000
     OutRangeA=1
     OutRangeB=0
    )
    Sim Space Settings_VelocityScaleZ:float=0
    Sim Space Settings_DampingAlpha:float=1
    Sim Space Settings_MaxLinearVelocity:float=500
    Sim Space Settings_MaxAngularVelocity:float=3
    Sim Space Settings_MaxLinearAcceleration:float=250
    Sim Space Settings_MaxAngularAcceleration:float=1000
    Sim Space Settings_ExternalLinearDragV:Vector="0,0,0"
    Sim Space Settings_ExternalLinearVelocity:Vector="0,0,0"
    Sim Space Settings_ExternalAngularVelocity:Vector="0,0,0"
   ) -> (then, Output_Get:Sim Space Settings Structure)
  ]}
]
variables [Sim Space Settings {
  type=SimSpaceSettings
  default {
   dampingAlpha=1
   maxAngularAcceleration=1000
   maxAngularVelocity=3
   maxLinearAcceleration=250
   maxLinearVelocity=500
   velocityScaleZ=0
   worldAlpha=1
   externalLinearDragV=<0 0 0>
  }
 }]
